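; Behavior tree injection evaluator for the custom intro of an ElbowChat NPC.
; On enter, it casts the behavior tree component's owner to a Controller and stores the controlled pawn in its Pawn variable.
; Its starting-condition check always returns true, and the file contains no further intro logic.
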(event EventOnEnter
  (params (OwnerComp :type Behavior%20Tree%20Component%20Object%20Reference))
  (bind _ascontroller (Utilities|Casting|CastToController (Components|GetOwner OwnerComp)))
  (Variables|Default|SetPawn (Pawn|GetControlledPawn _ascontroller)))

(fn EvaluateStartingConditions
  (params (OwnerComp :type Behavior%20Tree%20Component%20Object%20Reference))
  (return true))
